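; Building gameplay actor for thrown Thermite: a surface-attached burning area hazard that can explode when destroyed.
; BeginPlay initializes ThermiteNativeComponent and configures Niagara duration, radius and train flag from scalable floats and the attach parent.
; Native bloom, end and damage events drive effects, audio, sound indicators and a reversible timeline fading post-process.
; Server death triggers a multicast explosion gated by pre/post-bloom data rows; vault opening destroys the actor, while vault-start functions return false.
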(event EventBeginPlay
  (bind _self self)
  (bind _returnvalue (Actor|GetAttachParentActor _self))
  (bind (_asfort_vehicle_interface _bsuccess) (Utilities|Casting|CastToFortVehicleInterface _returnvalue))
  (bind _ns_thermite_fx_default (Variables|Default|GetNS_Thermite_FX_Default))
  ;[0.02442 0.15 0.03607 0.5] Begin Play
  (Networking|SwitchHasAuthority
    (:Authority
      ;;[1 1 1] Correct rotation fire FX (BGA spawns based on surface normal, but the static mesh becoming root component messed with visual orientation.)
      (Transformation|AddActorLocalRotation "90.000000,0.000000,-20.000000" false true))
    (:Remote
      _))
  (|CallParentFunction)
  (Audio|Components|Audio|Play (Variables|Default|GetLoopingAudio) 0.0)
  (bind _thermitenative (Variables|Default|GetThermiteNative)
    (:then
      (Class|ThermiteNativeComponent|InitializeThermite _thermitenative)
      ;;[1 0.3542 0.3542 0.5] VFX Adjustment
      (bind _returnvalue_2 (GameplayTags|HasMatchingGameplayTag(Message) _returnvalue "(TagName=\"Athena.Gameplay.Train\")"))
      (if _returnvalue_2
        (bind _output_get (Variables|Default|SetIsTrain _returnvalue_2))
        (Niagara|SetNiagaraVariable(Bool) _ns_thermite_fx_default "IsTrain" _output_get)
        (else
          (Niagara|SetNiagaraVariable(Bool) (Variables|Default|GetNS_Thermite_FX_Default_Inverse) "IsTrain" _bsuccess)))
      (Niagara|SetNiagaraVariable(Float) _ns_thermite_fx_default "Duration" (ScalableFloat|GetValueatLevel (Variables|Default|GetROW_AOE_Duration) 1.0))
      (Niagara|SetNiagaraVariable(Float) _ns_thermite_fx_default "Radius" (ScalableFloat|GetValueatLevel (Variables|Default|GetDangerRadius))))
    (:else)))

(event OnBloomStart(ThermiteNative)
  ;[0.06458 0.15 0.1034 0.5] Fade in and out for VFX, audio, and post processing
  (Variables|Default|SetIsFadingIn true)
  (Components|Activation|Deactivate (Variables|Default|GetNS_Thermite_FX_HeatingUp))
  (Components|Activation|Activate (Variables|Default|GetNS_Thermite_FX_Default_Inverse) true)
  (Components|Timeline|SetPlayRate (Variables|BGA_Thermite_Throwable|GetFadeSmokeTL) (Utilities|Reciprocal (ScalableFloat|GetValueatLevel (Variables|Default|GetROW_AOE_InitialDelay))))
  (Audio|Parameter|ExecuteTriggerParameter (Variables|Default|GetLoopingAudio) "OnBurning")
  (bind (_direction _alpha) (Timeline|FadeSmokeTL PlayFromStart)
    (:Update
      ;; Don't go to full blend weight so post can still be influenced by storm, time of day, etc
      (Variables|/Script/Engine.PostProcessComponent|SetBlendWeight (* _alpha 0.7) (Variables|Default|GetPostProcessComponent)))
    (:Finished
      (if (not (Variables|Default|GetIsFadingIn))
        (Rendering|SetVisibility (Variables|Default|GetStaticMesh) false true)
        (Rendering|SetActorHiddenInGame true)))))

(event OnThermiteEnd(ThermiteNative)
  ;[0.06458 0.15 0.1034 0.5] Fade in and out for VFX, audio, and post processing
  (Variables|Default|SetIsFadingIn false)
  (Components|Timeline|SetPlayRate (Variables|BGA_Thermite_Throwable|GetFadeSmokeTL) (Utilities|Reciprocal (Variables|Default|GetFadeOutTime)))
  (Audio|Components|Audio|FadeOut (Variables|Default|GetLoopingAudio) (Variables|Default|GetFadeOutTime) 0.0 "Linear" :!class /Script/Engine.AudioComponent)
  (Components|Activation|Deactivate (Variables|Default|GetNS_Thermite_FX_Default))
  (Timeline|FadeSmokeTL ReverseFromEnd))

(event Custom|Explode_Multicast
  (bind _returnvalue (Transformation|GetActorLocation))
  ;[0.15 0.04476 0.04103 0.5] Explosion
  (Utilities|FlowControl|DelayUntilNextTick)
  (bind _thermitenative (Variables|Default|GetThermiteNative)
    (:then
      (if (or (and (Variables|/Script/ThermiteGameplayRuntime.ThermiteNativeComponent|GetbHasBloomed _thermitenative) (Class|AthenaFunctionLibrary|RowtoBool (Variables|Explosion|GetbShouldExplodePostBloom))) (and (Class|AthenaFunctionLibrary|RowtoBool (Variables|Explosion|GetbShouldExplodePreBloom)) (not (Variables|/Script/ThermiteGameplayRuntime.ThermiteNativeComponent|GetbHasBloomed _thermitenative))))
        (Class|ThermiteNativeComponent|ExplodeThermite _thermitenative)
        (bind _returnvalue_2 (Niagara|SpawnSystematLocation "/ThermiteGameplay/Assets/FX/NS_Thermite_Explosion.NS_Thermite_Explosion" _returnvalue))
        (Niagara|SetNiagaraVariable(Bool) _returnvalue_2 "IsTrain" (Variables|Default|GetIsTrain))
        (Audio|PlaySoundatLocation (Variables|Default|GetExplosionSound) _returnvalue)
        (Fortnite|ShowSoundIndicatorLocalforTeam :Location (Transformation|GetActorLocation) :MaxAudibleDistance (Variables|/Script/Engine.SoundBase|GetMaxDistance (Variables|Default|GetExplosionSound)) :IndicatorType "Gunshot" :Team (Team|GetActorTeam (Projectile|GetInstigatorPlayerController)) :Affiliations (Utilities|Array|MakeArray "Friendly" "Hostile" "Neutral") :Tint "(R=0.000000,G=0.000000,B=0.000000,A=0.000000)" :AttenuationMaxRangeMultiplier 1.0 :AttenuationIntensityMultiplier 1.0)
        (else
          ;; Temp VFX: Spawn mini cosmetic explosion if the explosion fails to trigger
          (Niagara|SpawnSystematLocation :SystemTemplate "/ThermiteGameplay/Assets/FX/NS_Thermite_Explosion.NS_Thermite_Explosion" :Location _returnvalue :Scale "0.010000,0.010000,0.010000" :bAutoDestroy true :bAutoActivate true :PoolingMethod "None" :bPreCullCheck true))))
    (:else)))

(event EventOnDeathServer
  (params (Damage :type Float%20%28single-precision%29) (DamageTags :type Gameplay%20Tag%20Container%20Structure%20%28by%20ref%29) (Momentum :type Vector :default "0, 0, 0") (HitInfo :type Hit%20Result%20Structure%20%28by%20ref%29) (InstigatedBy :type Controller%20Object%20Reference) (DamageCauser :type Actor%20Object%20Reference) (EffectContext :type Gameplay%20Effect%20Context%20Handle%20Structure))
  ;[0.15 0.04476 0.04103 0.5] Explosion
  (if (Class|AthenaFunctionLibrary|RowtoBool (Variables|Explosion|GetbShouldExplodeOnDeath))
    (CallFunction|ExplodeMulticast)))

(event OnDamageApplied(ThermiteNative)
  ;[0.09701 0.0423 0.15 0.5] Damage FX
  (Fortnite|ShowSoundIndicatorLocalforTeam :Location (Transformation|GetActorLocation) :MaxAudibleDistance (Variables|/Script/Engine.SoundBase|GetMaxDistance (Variables|/Script/Engine.AudioComponent|GetSound (Variables|Default|GetLoopingAudio))) :IndicatorType "Gunshot" :Team (Team|GetActorTeam (Projectile|GetInstigatorPlayerController)) :Affiliations (Utilities|Array|MakeArray "Friendly" "Hostile" "Neutral") :Tint "(R=0.000000,G=0.000000,B=0.000000,A=0.000000)" :AttenuationMaxRangeMultiplier 1.0 :AttenuationIntensityMultiplier 1.0))

(event EventVaultOpened
  ;[0.05222 0.15 0.1006 0.5] Vault open reaction
  (Actor|DestroyActor))

(fn FilterByLOS
  (params (_Array :type Array%20of%20Actor%20Object%20References))
  (return 0))

(fn GetFirstValid
  (params (Array :type Array%20of%20Actor%20Object%20References))
  (for _array_element Array
    (Utilities|IsValid _array_element
      (:Is%20Valid
        (if (not (Variables|/Script/Engine.Actor|GetbHidden _array_element))
          (return _array_element))))))

(fn IsTrainVault
  (return false))

(fn IsVaultStarted
  (return false))

(fn RequestTrainVaultStart
  (params (Payload :type Struct%20Vivid%20Razor%20Core%20Vault%20Start%20Payload%20Structure))
  (return false))

(fn RequestVaultStart
  (params (Payload :type Struct%20Vivid%20Razor%20Core%20Vault%20Start%20Payload%20Structure))
  (return false))

(fn SetBloomingPosition
  (Variables|Default|SetDamageLocation_Back (+ (Transformation|GetActorLocation) (* (* (Transformation|GetActorUpVector) -1.0) 256.0)))
  (Variables|Default|SetDamageLocation_Front (+ (Transformation|GetActorLocation) (* (Transformation|GetActorUpVector) 256.0)))
  (return))

(fn ConstructionScript
  (|CallParentFunction))
