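; Thrown thermite projectile that can request a vault start on impact and initiates spawning of its gameplay actor (BGA).
; The authority-only hit override reverts invalid throws to pickups where flagged and calls InitiateSpawnBGA.
; For actors with the vault interface not yet started, it requests a train or regular vault start, destroying itself unless told to persist.
; Water impacts play splash VFX, temporary sound and a debug print.
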(event EventBeginPlay)

(event EventTick
  (params (DeltaSeconds :type Float%20%28single-precision%29)))

(event EventHeldWaterImpact
  (bind _returnvalue_location (Math|Transform|BreakTransform (Transformation|GetActorTransform)))
  (Development|PrintString "Hit Water Surface" true true "(R=0.000000,G=0.163821,B=1.000000,A=1.000000)")
  (Niagara|SpawnSystematLocation "/Game/Weapons/FORT_Generic_Effects/Impacts/Water/NS_Impact_Water_Athena/NS_Impact_Water_Beeg_Apollo.NS_Impact_Water_Beeg_Apollo" _returnvalue_location "90.000000,0.000000,0.000000")
  ; temp sfx
  (Audio|PlaySoundatLocation "/Game/Athena/Sounds/Weapons/Explosions/Expl_Debris_Water_Small_Cue.Expl_Debris_Water_Small_Cue" _returnvalue_location))

(event EventHandleHit
  (params (OtherActor :type Actor%20Object%20Reference) (OtherComp :type Primitive%20Component%20Object%20Reference) (Hit :type Hit%20Result%20Structure%20%28by%20ref%29) (HitLocation :type Vector :default "0, 0, 0"))
  (bind (_bblockinghit _binitialoverlap _time _distance _location _impactpoint _normal _impactnormal) (Collision|BreakHitResult Hit))
  ; Override so we can handle vaults
  (Networking|SwitchHasAuthority
    (:Authority
      (bind _asfort_pawn (Utilities|Casting|CastToFortPawn (Game|GetInstigator))
        (:then
          (Variables|Default|SetInstigatorFortPawn _asfort_pawn))
        (:CastFailed
          _))
      (bind (_validspawn _reverttopickup) (CallFunction|IsHitValidforSpawn OtherComp OtherActor Hit))
      (if _validspawn
        _
        (elif (Components|ComponentHasTag OtherComp "NoStick")
          _
          (elif _reverttopickup
            (CallFunction|RevertThrownObjecttoPickup Hit))))
      (if (Utilities|DoesObjectImplementInterface OtherActor "/VividRazorCoreGameplay/Gameplay/Interfaces/BPI_VividRazorCore_VaultEvents.BPI_VividRazorCore_VaultEvents_C")
        (bind _isvaultstarted (Class|BPIVividRazorCoreVaultEvents|IsVaultStarted(Message) OtherActor))
        (if (not _isvaultstarted)
          (bind _preservetriggeringactor (CallFunction|TryRequestVaultStart OtherActor (Utilities|Struct|MakeStructVividRazorCoreVaultStartPayload "(TagName=\"\")" (Utilities|ResolveSoftReference (Variables|Default|GetInstigatorFortPawn)) (Utilities|ResolveSoftReference self) _impactnormal Hit)))
          (if (not _preservetriggeringactor)
            (Rendering|SetActorHiddenInGame false)
            (Collision|SetActorEnableCollision false)
            (Game|Components|ProjectileMovement|StopSimulating (Variables|Movement|GetProjectileMovementComponent) Hit)
            (Actor|DestroyActor)))
        (CallFunction|InitiateSpawnBGA Hit OtherActor OtherComp HitLocation)
        (else
          (CallFunction|InitiateSpawnBGA Hit OtherActor OtherComp HitLocation))))))

(fn TryRequestVaultStart
  (params (VaultActor :type Object%20Reference) (Payload :type Struct%20Vivid%20Razor%20Core%20Vault%20Start%20Payload%20Structure))
  (bind _result (Class|BPIVividRazorCoreVaultEvents|IsTrainVault(Message) VaultActor))
  (if _result
    (bind _preservetriggeringactor (Class|BPIVividRazorCoreVaultEvents|RequestTrainVaultStart(Message) VaultActor Payload))
    (return _preservetriggeringactor)
    (else
      (bind _preservetriggeringactor_2 (Class|BPIVividRazorCoreVaultEvents|RequestVaultStart(Message) VaultActor Payload))
      (return _preservetriggeringactor_2))))

(fn ConstructionScript
  (|CallParentFunction))
