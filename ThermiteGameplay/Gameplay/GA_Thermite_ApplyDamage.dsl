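; Thermite gameplay ability that, when activated from a gameplay event, deals damage immediately and then repeatedly on a looping timer.
; The timer interval comes from a tick-interval scalable float, and its handle is stored in DamageTimerHandle.
; Each tick runs target selection with the Gameplay.Application.Explode tag and applies the gameplay effect container to the results.
; The custom ability source transform is overridden to use the avatar actor's transform.
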(event Ability|EventActivateAbilityFromEvent
  (params (EventData :type Gameplay%20Event%20Data%20Structure%20%28by%20ref%29))
  ;[1 1 1] Apply damage effect once, then kick off a timer to periodically apply the effect
  (bind _returnvalue (Utilities|Time|SetTimerbyEvent (delegate Custom|ApplyDamage) (ScalableFloat|GetValueatLevel (Variables|Default|GetRow_TickInterval)) true))
  (Variables|Default|SetDamageTimerHandle _returnvalue)
  (CallFunction|ApplyDamage))

(event Custom|ApplyDamage
  (bind (_asynctaskproxy _targetdata _applicationtag) (Ability|Tasks|DoTargetSelection "None" "(TagName=\"Gameplay.Application.Explode\")" true))
  (Ability|ApplyGameplayEffectContainer _targetdata _applicationtag 0))

(fn GetCustomAbilitySourceTransform
  (return (Transformation|GetActorTransform (Ability|GetAvatarActorfromActorInfo))))
