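; Secondary throw ability for the Thermite item, a child ability that leaves activation to its parent and adds a leftover "Hello" debug print.
; It overrides the trajectory validity check to reject any location where the hit actor is a Fortnite water body, using a soft class check.
; The override calls the parent check but ignores its result, so every non-water hit counts as valid.
; The end-ability event is only a disabled stub.
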(event Ability|EventActivateAbility
  (|CallParentFunction)
  (Development|PrintString "Hello"))

(event Ability|EventOnEndAbility
  (params (bWasCancelled :type Boolean))
  (|CallParentFunction bWasCancelled) ; This node is disabled and will not be called.\nDrag off pins to build functionality.
)

(fn IsTrajectoryLocationValid
  (params (SplinePoints :type Array%20of%20Vectors) (SplineTangents :type Array%20of%20Vectors) (TrajectoryHitResult :type Hit%20Result%20Structure))
  (bind (_bblockinghit _binitialoverlap _time _distance _location _impactpoint _normal _impactnormal _physmat _hitactor) (Collision|BreakHitResult TrajectoryHitResult))
  (|CallParentFunction SplinePoints SplineTangents TrajectoryHitResult)
  (Utilities|IsA(soft) _hitactor "/Script/FortniteGame.FortWaterBodyActor"
    (:True
      (return false))
    (:False
      (return true))))
