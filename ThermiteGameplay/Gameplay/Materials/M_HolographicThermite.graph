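# Unlit translucent hologram material for the thermite charge, supporting static, Nanite and skeletal meshes.
# The HasThermite custom primitive data value blends active and disabled colors, tinted by a sweep color and pulsed over time.
# Fresnel or sphere mask opacity combines with animated scanlines, a moving sheen and optional depth-faded masking.
# Optional edge glow reads UV1 and distance fields, with a mobile fallback; emissive is compensated for eye adaptation.
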Material /ThermiteGameplay/Gameplay/Materials/M_HolographicThermite
graph {
 properties {
  bUsedWithMeshDeformer=true
  bUsedWithNanite=true
  bUsedWithSkeletalMesh=true
  blendMode=BLEND_Translucent
  previewMesh=/ThermiteGameplay/Gameplay/Meshes/ThermiteCharge
  shadingModel=MSM_Unlit
  usedShadingModels=Unlit
 }
 nodes [
  ComponentMask(None=WorldPosition().XYZ) {b=true}
  n1 = ScalarParameter() {
   defaultValue=1.68
   desc=Eye Adaptation in the old scene was 1.68 - scale to match old appearance
   group=ThermiteHologram
   parameterName=FX Emissive EyeAdaptation Prescale
   sliderMax=4
  }
  n2 = LinearInterpolate(
   B=DistanceToNearestSurface!(World Position)
   Alpha=ScalarParameter() {defaultValue=1, group=ThermiteHologram, parameterName=EdgeGlow_UseDistanceField}
  ) {constA=100000, constB=1000000}
  n3 = ComponentMask(None=TextureCoordinate() {coordinateIndex=1}) {r=true}
  n4 = FeatureLevelSwitch(Default=DistanceFieldsRenderingSwitch!(No=@n3, Yes=Min(A=@n2, B=@n3)), ES3_1=@n3)
  n5 = Constant()
  n6 = ScalarParameter() {defaultValue=5, group=ThermiteHologram, parameterName=EdgeGlow_Radius}
  n7 = ScalarParameter() {defaultValue=1, group=ThermiteHologram, parameterName=EdgeGlow_Hardness}
  n8 = StaticSwitchParameter(
   True=Multiply(
    A=SphereMask(A=@n4, B=@n5, Radius=@n6, Hardness=@n7)
    B=ScalarParameter() {defaultValue=0.6, group=ThermiteHologram, parameterName=Edge Glow Intensity}
   )
   False=Constant()
  ) {
   desc="Doc: Open Parent material and copy link from there.\r\nhttps://confluence.it.epicgames.com/pages/viewpage.action?pageId=251430673"
   group=ThermiteHologram
   parameterName="UseEdgeGlow(ReadDescription)"
  }
  n9 = Time()
  n10 = Multiply(A=Sine(None=Multiply(A=@n9) {constB=1.8})) {constB=0.25}
  n11 = LinearInterpolate(
   A=VectorParameter() {parameterName=Color_Active, defaultValue=<rgba 0.0286 0.2739 1.5 0>}.RGB
   B=VectorParameter() {parameterName=Color_Disabled, defaultValue=<rgba 1.5 0.127 0.0547 1>}.RGB
   Alpha=ScalarParameter() {bUseCustomPrimitiveData=true, parameterName=HasThermite}
  )
  n12 = Multiply(A=@n11, B=VectorParameter() {group=ThermiteHologram, parameterName=SweepColour, defaultValue=<rgba 10 7.1507 2.8299 1>}.RGB)
  n13 = NamedRerouteUsage()
  n14 = Constant() {r=0.23}
  n15 = Constant() {r=0.5}
  n16 = ScalarParameter() {defaultValue=2, group=ThermiteHologram, parameterName=FresnelExponent}
  n17 = LinearInterpolate(
   A=VertexNormalWS()
   B=Transform(None=TextureSampleParameter2D() {parameterName=Normal, samplerType=SAMPLERTYPE_Normal, texture=&FlatNormalBlue}.RGB)
   Alpha=ScalarParameter() {group=ThermiteHologram, parameterName=UseNormalTexture}
  )
  n18 = MaterialFunctionCall(Exponent=@n16, Vertex / Pixel Normal=@n17) {materialFunction=&MF_Fresnel_FX}
  n19 = DotProduct(A=@n17, B=CameraVectorWS())
  n20 = Power(Base=OneMinus(None=Abs(None=@n19)), Exp=@n16)
  n21 = ScalarParameter() {parameterName=UseFresnelforTwoSide}
  n22 = LinearInterpolate(
   A=Add(
    A=LinearInterpolate(A=@n18.Fresnel Inverted, B=@n20, Alpha=@n21)
    B=ScalarParameter() {defaultValue=0.32, group=ThermiteHologram, parameterName=Min Opacity}
   )
   B=NamedRerouteUsage()
   Alpha=ScalarParameter() {group=ThermiteHologram, parameterName=ReplaceFresnelWithSphereMask, sortPriority=-10}
  )
  n23 = Multiply(A=NamedRerouteUsage()) {constB=3}
  n24 = LocalPosition()
  n25 = ComponentMask(None=ObjectLocalBounds().Extents) {b=true}
  n26 = Time()
  n27 = ScalarParameter() {defaultValue=32, group=ThermiteHologram, parameterName=ScanlineTileHorizontal}
  n28 = Multiply(A=Add(A=Divide(A=@n24.Z, B=@n25), B=Multiply(A=@n26) {constB=5}), B=@n27)
  n29 = Floor(None=Add(A=Frac(None=@n28)) {constB=0.5})
  n30 = Max(A=NamedRerouteUsage(), B=Add(A=Add(A=@n22, B=@n23), B=Multiply(A=@n29) {constB=0.03}))
  n31 = Multiply(
   A=Add(A=NamedRerouteDeclaration(None=@n8) {name=Edge Glow, nodeColor=<rgba 0.125 0 1 1>}, B=Add(A=@n10))
   B=LinearInterpolate(B=LinearInterpolate(A=@n11, B=@n12, Alpha=@n13), Alpha=SmoothStep(Min=@n14, Max=@n15, Value=@n30))
  )
  n32 = Multiply(A=@n31, B=@n1) {constB=1.3}
  ComponentMask(None=@n32) {b=true, g=true, r=true}
  Constant2Vector()
  Constant() {r=0.28}
  Constant()
  n33 = VertexInterpolator(VS=PreSkinnedPosition())
  n34 = Divide(
   A=ComponentMask(None=Transform(None=@n33.PS) {transformSourceType=TRANSFORMSOURCE_Local, transformType=TRANSFORM_View}) {g=true}
   B=ScalarParameter() {defaultValue=3, group=ThermiteHologram, parameterName=ScreenScapeScanlineTileHorizontal}
  )
  n35 = Multiply(A=Time() {bOverride_Period=true, period=0.2}) {constB=5}
  n36 = Constant() {r=1}
  n37 = Panner(Coordinate=Multiply(
    A=TextureCoordinate()
    B=MaterialFunctionCall(X=@n36, Y=@n27) {materialFunction=&MakeFloat2}.Result
   )) {speedY=-5}
  Frac(None=StaticSwitchParameter(True=Subtract(A=@n34, B=@n35), False=ComponentMask(None=@n37) {g=true}) {group=ThermiteHologram, parameterName=PseudoScreenSpaceScanlines})
  n38 = Subtract(A=WorldPosition().XYZ, B=ObjectPositionWS())
  Multiply(A=@n38, B=ScalarParameter() {defaultValue=0.0025, group=ThermiteHologram, parameterName=WorldTileScale})
  n39 = Normalize(VectorInput=Transform(None=VectorParameter() {group=ThermiteHologram, parameterName=SweepDirectionCamSpace, defaultValue=<rgba -0.5 0.45 -1 1>}.RGB) {transformSourceType=TRANSFORMSOURCE_Camera})
  n40 = ComponentMask(None=Multiply(A=DotProduct(A=@n38, B=@n39), B=@n39)) {b=true, r=true}
  n41 = ObjectPositionWS()
  n42 = Add(A=Time(), B=Add(A=ComponentMask(None=@n41) {g=true}, B=ComponentMask(None=@n41) {r=true}))
  n43 = Constant() {r=0.28}
  n44 = Frac(None=ComponentMask(None=Panner(Coordinate=@n40, Time=@n42, Speed=@n43) {speedX=0.2, speedY=0.2}) {g=true})
  n45 = Add(A=@n44) {constB=0.1}
  n46 = OneMinus(None=Multiply(A=Subtract(A=@n44) {constB=0.9}) {constB=10})
  n47 = MaterialFunctionCall() {materialFunction=&MF_Fresnel_FX}
  Multiply(A=Multiply(A=Floor(None=@n45), B=Power(Base=@n46) {constExponent=3}), B=Saturate(None=Add(A=@n47.Fresnel Inverted) {constB=0.1}))
  n48 = DepthFade(FadeDistance=Constant() {r=16})
  Multiply(A=@n31, B=Add(A=OneMinus(None=@n48)))
  n49 = SphereMask(
   A=WorldPosition().XYZ
   B=TransformPosition(None=VectorParameter() {group=ThermiteHologram, parameterName=SphereMaskPivot, defaultValue=<rgba 0 0 0 1>}.RGB) {transformType=TRANSFORMPOSSOURCE_World}
   Radius=ScalarParameter() {defaultValue=100, group=ThermiteHologram, parameterName=SphereMaskFalloff}
   Hardness=ScalarParameter() {group=ThermiteHologram, parameterName=SphereMaskHardness}
  )
  n50 = OneMinus(None=@n49)
  n51 = ScalarParameter() {group=ThermiteHologram, parameterName=InvertSphereMask}
  NamedRerouteDeclaration(None=Power(
    Base=LinearInterpolate(A=@n50, B=@n49, Alpha=@n51)
    Exp=ScalarParameter() {defaultValue=1, group=ThermiteHologram, parameterName=SphereMaskExponent}
   )) {name=SphereMaskOpacity, nodeColor=<rgba 1 0.1172 0 1>}
  NamedRerouteDeclaration(None=EyeAdaptationInverse(
    LightValueInput=@n32
    AlphaInput=ScalarParameter() {defaultValue=1, desc=Mix to Original, group=ThermiteHologram, parameterName=FX Emissive EyeAdaptation Alpha}
   ).EyeAdaptationInverse) {name=Emisive, nodeColor=<rgba 0.1953 1 0 1>}
  n52 = ComponentMask(None=TextureCoordinate()) {g=true}
  n53 = Constant()
  n54 = ScalarParameter() {defaultValue=1, group=ThermiteHologram, parameterName=MaskSize}
  n55 = DepthFade(
   Opacity=Multiply(A=@n30, B=SphereMask(A=@n52, B=@n53, Radius=@n54) {hardnessPercent=0})
   FadeDistance=ScalarParameter() {group=ThermiteHologram, parameterName=DepthFade}
  )
  n56 = ScalarParameter() {group=ThermiteHologram, parameterName=DepthFade_ground}
  n57 = Multiply(A=OneMinus(None=DepthFade(FadeDistance=@n56)), B=Constant() {r=0.7})
  NamedRerouteDeclaration(None=LinearInterpolate(A=@n30, B=Add(A=@n55, B=@n57), Alpha=ScalarParameter() {parameterName=Mask})) {name=Opacity, nodeColor=<rgba 1 0 0.5156 1>}
  n58 = ScalarParameter() {defaultValue=2, group=ThermiteHologram, parameterName=FresnelExponent}
  n59 = ScalarParameter() {defaultValue=0.1, parameterName=Sheen Size}
  n60 = Constant()
  n61 = Normalize(VectorInput=VectorParameter() {group=ThermiteHologram, parameterName=SweepDirectionCamSpace, defaultValue=<rgba -0.5 0.45 -1 1>}.RGB)
  n62 = DotProduct(A=LocalPosition().XYZ, B=@n61)
  n63 = Panner(
   Coordinate=ComponentMask(None=Multiply(A=@n62, B=@n61)) {b=true, r=true}
   Time=Time()
   Speed=ScalarParameter() {defaultValue=100, parameterName=Speed}
  ) {speedX=0.2, speedY=0.2}
  n64 = Frac(None=Divide(A=ComponentMask(None=@n63) {g=true}, B=ScalarParameter() {defaultValue=200, group=ThermiteHologram, parameterName=Freq}))
  n65 = Multiply(
   A=NamedRerouteDeclaration(None=Fresnel(ExponentIn=@n58) {baseReflectFraction=0.01}) {name=Fresnel_simple, nodeColor=<rgba 0 0.5547 1 1>}
   B=NamedRerouteDeclaration(None=SmoothStep(Min=@n59, Max=@n60, Value=@n64)) {name=M_Sheen, nodeColor=<rgba 0 1 0.4141 1>}
  )
  NamedRerouteDeclaration(None=@n65) {name=M_Sheen_fresnel, nodeColor=<rgba 0 0.8359 1 1>}
  NamedRerouteUsage()
  NamedRerouteUsage()
  NamedRerouteUsage()
  NamedRerouteUsage()
  Reroute()
  Reroute()
  Reroute(None=@n65)
 ]
}
