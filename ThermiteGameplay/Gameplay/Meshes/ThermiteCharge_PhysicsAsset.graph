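# Physics asset for the thermite charge mesh in ThermiteGameplay, giving it a simple collision and simulation shape.
# It holds a single body on the root bone made of one capsule, slightly offset from the bone origin and rotated onto its side.
# It defines no constraints, so the charge acts as one rigid piece.
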PhysicsAsset /ThermiteGameplay/Gameplay/Meshes/ThermiteCharge_PhysicsAsset
bodies {root [{shapeType=Capsule, length=19.2424, radius=10.7597, center=<0.5662 -0.1137 -0.5529>, rotation=<rot -9.5949 -90 -3.3627>}]}
constraints []
